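# Niagara system for the swing trail of the FumeFleeceClap pickaxe, drawn as a short-lived pale cyan ribbon.
# A single Trail emitter uses anim-trail functions to spawn, initialize and update particles from skeletal mesh bones.
# Alpha multiplies a time-of-day-boosted lifetime fade by a system-age curve, scaled by front-end or gameplay context and zeroed by a swing-section comparison.
# A time-of-day color tints the RGB.
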NiagaraSystem /BRCosmetics/Weapons/FORT_Melee/Pickaxe_FumeFleeceClap/FX/NS_Pickaxe_FumeFleeceClap_Trail
dependencies {
 usedDataInterfaces [&NiagaraDataInterfaceCurve, &NiagaraDataInterfaceSkeletalMesh, &NiagaraDataInterfaceColorCurve]
 usedDynamicInputs [
  &MakeVector4_FromLinearColor
  &self:SectionIndexFloatSelector
  &ColorFromCurve
 ]
 usedModules [
  &self:SystemSpawnScript.SystemScriptSource.SystemScriptGraph.NiagaraNodeAssignment_0.SetVariables_#1
  &SystemState
  &self:SystemSpawnScript.SystemScriptSource.SystemScriptGraph.NiagaraNodeAssignment_1.SetVariables_#2
  &self:SystemSpawnScript.SystemScriptSource.SystemScriptGraph.NiagaraNodeAssignment_2.SetVariables_#3
 ]
}
emitters {Trail {
  emitterUpdateScript [
   &EmitterState(Life Cycle Mode=::NewEnumerator1, Loop Behavior=::NewEnumerator1, Loop Duration=0.3, UseLoopDelay=false)
   &FN_SpawnAnimTrail()
  ]
  particleSpawnScript [
   &FN_InitializeAnimTrailBoneIndecies()
   &FN_InitializeAnimTrail(Color=<rgba 0.4531 0.8431 1 1>, Lifetime=0.3, SkeletalMesh=$System.Skeletal Mesh)
  ]
  particleUpdateScript [
   &ParticleState()
   &FN_UpdateAnimTrail(SkeletalMesh=$System.Skeletal Mesh)
   &ScaleColor(
    Scale Alpha=&Multiply_Float(
     A=&FloatFromCurve(
      FloatCurve=DataInterface() {propertyValues {Curve {keys [
          {arriveTangent=0.0666, interpMode=RCIM_Cubic, leaveTangent=0.0666, tangentMode=RCTM_User, time=0, value=1}
          {interpMode=RCIM_Cubic, tangentMode=RCTM_User, time=1, value=0}
         ]}}}
      Scale Curve=&FNTimeOfDayScalar(Night Scale=2, Morning Scale=1.2, Evening Scale=1.2)
     )
     B=&FloatFromCurve(
      FloatCurve=DataInterface() {propertyValues {Curve {
         defaultValue=1
         keys [
          {arriveTangent=0.074, interpMode=RCIM_Cubic, leaveTangent=0.074, tangentMode=RCTM_User, time=0, value=1}
          {arriveTangent=-8.6402, interpMode=RCIM_Cubic, leaveTangent=-8.6402, tangentMode=RCTM_User, time=0.1972, value=0.5421}
          {arriveTangent=0.0063, interpMode=RCIM_Cubic, leaveTangent=0.0063, tangentMode=RCTM_User, time=0.26, value=0}
         ]
        }}}
      CurveIndex=$Engine.System.Age
      Scale Curve=&MakeCustomFloatFromBool(
       Bool=&SetBoolByIntComparison(A=$User.SwingSectionIndex, B=0, Comparison Type=::NewEnumerator4)
       True Float=0
       False Float=&FN_FrontEndFloatModifier(Gameplay=0.01, Front End=0.075)
      )
     )
    )
    Scale RGB=&MakeVectorFromLinearColorRGB(Color=$System.TODColor)
    ScaleA=true
    ScaleRGB=true
   ) {module=ScaleColor001}
  ]
  rendererClasses [&NiagaraRibbonRendererProperties]
  renderers [{rendererClass=&NiagaraRibbonRendererProperties}]
 }}
summary {
 systemName=NS_Pickaxe_FumeFleeceClap_Trail
 emitters [{emitterName=Trail, rendererClasses [&NiagaraRibbonRendererProperties]}]
 userVariables [
  User.bIsFrontEnd {type=NiagaraBool}
  User.bIsFrontEndPreview {type=NiagaraBool}
  User.RibbonIDOffset {type=NiagaraInt32}
  User.SwingSectionIndex {type=NiagaraInt32}
 ]
}
